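# Physics asset for the Tomato1 vehicle cosmetic body, defining its simplified collision shape.
# A single body on the root bone holds one box named root_box, raised above the bone origin to enclose the body as one rough volume.
# It has no constraints, so it describes one rigid body with no articulated parts.
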PhysicsAsset /VehicleCosmetics/Bodies/Tomato1/Meshes/Body_Tomato1_PhysicsDM
bodies {root [{shapeType=Box, extentX=444, extentY=216, extentZ=132, shapeName=root_box, center=<0 0 103>}]}
constraints []
