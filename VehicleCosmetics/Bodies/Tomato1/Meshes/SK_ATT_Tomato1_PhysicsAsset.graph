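# Physics asset for the SK_ATT_Tomato1 skeletal mesh among the Tomato1 vehicle body cosmetics.
# It lists no constraints and shows no physics bodies, so it sets up no jointed or ragdoll simulation.
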PhysicsAsset /VehicleCosmetics/Bodies/Tomato1/Meshes/SK_ATT_Tomato1_PhysicsAsset
constraints []
